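# Translucent unlit UI material for a loading throbber, apparently an animated ring with a two-color tint.
# A horizontal linear gradient blends the color1 and color2 parameters.
# A time-driven Rotator feeds a radial angle conversion, blended by Solid_Animate and wrapped with Frac to form the spinning sweep.
# Inner and outer exponential radial gradients confine the sweep to a ring before a reflected-gradient function shapes the mask, which OpacityOverride adjusts.
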Material /FortUILibrary/Systems/Throbber/M_V2Radio_Loading
graph {
 properties {
  bUsedWithStaticMesh=false
  blendMode=BLEND_Translucent
  materialDomain=MD_UI
  pixelDepthOffsetMode=PDOM_Legacy
  shadingModel=MSM_Unlit
  usedShadingModels=Unlit
 }
 nodes [
  LinearInterpolate(
   A=VectorParameter() {parameterName=color1, defaultValue=<rgba 1 0.4852 0.0595 1>}.RGB
   B=VectorParameter() {parameterName=color2, defaultValue=<rgba 0 1 0.9843 1>}.RGB
   Alpha=Power(Base=MaterialFunctionCall() {materialFunction=&LinearGradient}.UGradient) {constExponent=1}
  )
  n1 = ScalarParameter() {defaultValue=24, group=Fill, parameterName=Speed}
  n2 = Rotator(Coordinate=TextureCoordinate(), Time=Multiply(A=Time(), B=Multiply(A=@n1) {constB=-1})) {speed=0.21}
  n3 = LinearInterpolate(
   A=MaterialFunctionCall(Vector or UVs=Add(A=@n2) {constB=-0.5}) {materialFunction=&VectorToRadialValue}.Vector Converted to Angle
   Alpha=ScalarParameter() {defaultValue=-1, group=Fill, parameterName=Solid_Animate, sliderMax=1}
  )
  n4 = OneMinus(None=MaterialFunctionCall(
    UVs=TextureCoordinate()
    Radius=ScalarParameter() {defaultValue=0.2, group=Circle, parameterName=InnerCircleRadius}
    Density=ScalarParameter() {defaultValue=20, group=Circle, parameterName=InnerCircleSharpness}
   ) {materialFunction=&RadialGradientExponential}.RadialGradientExponential)
  n5 = MaterialFunctionCall(UVs=TextureCoordinate(), Density=ScalarParameter() {defaultValue=20, group=Circle, parameterName=OuterCircleSharpness}) {materialFunction=&RadialGradientExponential}
  n6 = MaterialFunctionCall!(
   UVs=LinearInterpolate(B=Frac(None=@n3), Alpha=LinearInterpolate(B=@n4, Alpha=@n5.RadialGradientExponential))
   Offset
   RotationCenter
   RotationAngle
  ) {materialFunction=/FortUILibrary/MaterialFunctions/Gradients/MF_UI_ReflectedGradient}
  Multiply(A=LinearInterpolate(A=ScalarParameter() {defaultValue=-1, parameterName=OpacityOverride}, Alpha=Multiply(A=@n6.Result) {constB=2}) {constB=1.2}) {constB=0.5}
 ]
}
